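# Niagara system for the FangGang Sparks bass cosmetic, drawing two glowing red bat-eye sprites that pulse.
# One burst spawns both looping particles, with partition index selecting each eye's position offset and facing.
# A color curve brightens red on the audio BeatLifetime parameter, or on normalized age when User.bIsFrontEndPreview is set.
# A separate curve swells sprite size, scaled by owner scale, and a camera offset is applied.
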NiagaraSystem /SparksCosmetics/Instrument/Bass/FangGang/Effects/NS_Bass_FangGang
dependencies {
 usedDataInterfaces [&NiagaraDataInterfaceArrayFloat3, &NiagaraDataInterfaceVectorCurve, &NiagaraDataInterfaceCurve]
 usedDynamicInputs []
 usedModules [&SystemState]
}
emitters {BatEyes {
  emitterUpdateScript [
   &EmitterState(Life Cycle Mode=::NewEnumerator1, Loop Behavior=::NewEnumerator1, Loop Duration Mode=::NewEnumerator1, UseLoopDelay=false)
   &SpawnBurst_Instantaneous(Spawn Count=2, Use Spawn Probability=false, UseLoopCountLimit=false)
  ]
  particleSpawnScript [
   &PartitionParticles()
   &InitializeParticle(
    Color=<rgba 1 0.1 0 1>
    Position Offset=&SelectVectorFromArray(
     Direct Array Index=$Particles.PartitionParticles.Partition
     Vector Selection Array=DataInterface() {propertyValues {FloatData [<-1.65 1.7 -17.6>, <1.65 1.8 -17.6>]}}
    )
    Sprite Size Mode=::NewEnumerator3
    Uniform Sprite Size=5
    UsePositionOffset=true
    Write Material Random=true
   )
  ]
  particleUpdateScript [
   &ParticleState(Kill Particles When Lifetime Has Elapsed=false, Loop Particles Lifetime=true)
   &ScaleColor(
    Scale RGB=&VectorFromCurve(
     VectorCurve=DataInterface() {propertyValues {
       XCurve {keys [
         {arriveTangent=9.0725, interpMode=RCIM_Cubic, leaveTangent=9.0725, tangentMode=RCTM_User, time=0, value=1}
         {arriveTangent=-2.5249, interpMode=RCIM_Cubic, leaveTangent=-2.5249, tangentMode=RCTM_User, time=0.1, value=2}
         {arriveTangent=-0.0405, interpMode=RCIM_Cubic, leaveTangent=-0.0405, tangentMode=RCTM_User, time=1, value=1}
        ]}
       YCurve {keys [<at 0 1>,<at 1 1>]}
       ZCurve {keys [<at 0 1>,<at 1 1>]}
      }}
     CurveIndex=&MakeCustomFloatFromBool(Bool=$User.bIsFrontEndPreview, True Float=$Particles.NormalizedAge, False Float=$NPC.NPC_Sparks_Audio_Parameters001.BeatLifetime)
    )
    ScaleA=true
    ScaleRGB=true
   )
   &ScaleSpriteSize(
    Uniform Curve Scale=&MakeFloatFromVector(Vector=$Engine.Owner.Scale, Channel=::NewEnumerator0)
    Uniform Curve Sprite Scale=DataInterface() {propertyValues {Curve {keys [
        {arriveTangent=4.0309, interpMode=RCIM_Cubic, leaveTangent=4.0309, leaveTangentWeight=0.3333, tangentMode=RCTM_User, time=0, value=1}
        {
         arriveTangent=0.065
         arriveTangentWeight=0.0333
         interpMode=RCIM_Cubic
         leaveTangent=0.065
         leaveTangentWeight=0.3
         tangentMode=RCTM_User
         time=0.2
         value=1.5
        }
        {arriveTangent=-0.0405, arriveTangentWeight=0.3, interpMode=RCIM_Cubic, leaveTangent=-0.0405, tangentMode=RCTM_User, time=1, value=1}
       ]}}}
   )
   &SpriteFacingAndAlignment(
    Alignment Coordinate Space=::World
    Sprite Alignment=<-1 0 0>
    Sprite Facing=&SelectVectorFromArray(
     Direct Array Index=$Particles.PartitionParticles.Partition
     Vector Selection Array=DataInterface() {propertyValues {FloatData [<-0.5 1 0>, <0.5 1 0>]}}
    )
    WriteAlignment=true
    WriteFacing=true
   )
   &CameraOffset(Camera Offset Amount=1)
  ]
  rendererClasses [&NiagaraSpriteRendererProperties]
  renderers [{rendererClass=&NiagaraSpriteRendererProperties}]
 }}
summary {
 systemName=NS_Bass_FangGang
 emitters [{emitterName=BatEyes, rendererClasses [&NiagaraSpriteRendererProperties]}]
 userVariables [User.bIsFrontEndPreview {type=NiagaraBool}]
}
